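# Sound cue that plays when the Mechanical Parts crafting ingredient is picked up.
# A Random node picks one of two equally weighted pickup waves and feeds a Modulator that varies pitch slightly while holding volume steady.
# The cue plays at reduced volume in the loot-collect sound class and borrows the weapon-drop attenuation asset for distance falloff.
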SoundCue /PrimalGameplay/Sounds/Crafting_Ingredients/MechanicalParts/MechanicalParts_Pickup_Cue
root=Modulator_0
nodes [
 {id=Modulator_0, children [Random_1], properties {pitchMin=0.97, volumeMax=1, volumeMin=1}}
 {id=Random_1, children [WavePlayer_0,WavePlayer_3], properties {weights [1,1]}}
 {id=WavePlayer_0, properties {soundWaveAssetPtr=&/MechanicalParts_Pickup_01}}
 {id=WavePlayer_3, properties {soundWaveAssetPtr=&/MechanicalParts_Pickup_02}}
]
properties {
 attenuationSettings=&Fort_Weapon_Drop_Att
 maxDistance=3000
 soundClassObject=&Fort_Loot_Collect
 volumeMultiplier=0.65
}
